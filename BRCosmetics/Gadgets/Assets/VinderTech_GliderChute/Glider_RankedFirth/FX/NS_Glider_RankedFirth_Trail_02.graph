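# Niagara trail effect for the RankedFirth glider, combining socket-anchored ribbons, a glow and speed lines.
# Burst-spawned ribbon points form an arc at FX_Socket_02, with index-based color, fade and width and a time-scrolled wave, drawn by two renderers.
# Ribbon_lowmed repeats this with fewer points, while BlowingParticles001 pins a sprite and light to the sockets, tinted by time of day.
# SpeedLine emits velocity-stretched streaks, with spawning zeroed in front-end preview.
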NiagaraSystem /BRCosmetics/Gadgets/Assets/VinderTech_GliderChute/Glider_RankedFirth/FX/NS_Glider_RankedFirth_Trail_02
dependencies {
 usedDataInterfaces [
  &NiagaraDataInterfaceSkeletalMesh
  &NiagaraDataInterfaceColorCurve
  &NiagaraDataInterfaceCurve
  &NiagaraDataInterfaceVector2DCurve
 ]
 usedDynamicInputs [&SetBoolByFloatComparison, &FN_FrontEndFloatModifier]
 usedModules [&SystemState, &Timeline]
}
emitters {
 BlowingParticles001 {
  emitterUpdateScript [
   &EmitterState(Life Cycle Mode=::NewEnumerator1, Loop Behavior=::NewEnumerator1, Loop Duration Mode=::NewEnumerator1, UseLoopDelay=false)
   &SpawnBurst_Instantaneous(Use Spawn Probability=false, UseLoopCountLimit=false)
  ]
  particleSpawnScript [
   &InitializeParticle(
    Color=<rgba 0.5033 0.1067 0.6514 1>
    Lifetime=1.5
    Ribbon Width=3
    Ribbon Width Mode=::NewEnumerator1
    Sprite Rotation Mode=::NewEnumerator2
    Sprite Size Mode=::NewEnumerator3
    Uniform Sprite Size=30
    UsePositionOffset=false
    Write Material Random=true
   )
   &SkeletalMeshLocation(
    MaskSpawnGroup=false
    Mesh Sampling Type=::NewEnumerator5
    Orientation Sampling=::NewEnumerator1
    OverrideRandomSeed=false
    Skeletal Mesh=DataInterface() {propertyValues {
      DefaultMesh=null
      ExcludeBoneName=root
      OptionalComponentClass=null
      PreviewMesh=/BRCosmetics/Gadgets/Assets/VinderTech_GliderChute/Glider_Season_32/Meshes/Season_32_Glider
      SkinningMode=SkinOnTheFly
      SoftSourceActor=null
      SourceMode=Default
      UvSetIndex=0
      WholeMeshLOD=-1
      bExcludeBone=false
      bReadDeformedGeometry=true
      bRequireCurrentFrameData=true
      ComponentTags []
      FilteredBones []
      FilteredSockets [FX_Socket_01]
      MeshUserParameter {parameter {name=None, typeDefHandle {registeredTypeIndex=1}, varData []}}
      SamplingRegions []
     }}
    Use Sampled Mesh Attribute as Ribbon ID=false
   ) {module=SkeletalMeshLocation004}
   &CameraOffset()
  ]
  particleUpdateScript [
   &ParticleState(
    Kill Particles When Lifetime Has Elapsed=false
    Let Infinitely Lived Particles Die When Emitter Deactivates=true
    Loop Particles Lifetime=true
   )
   &SkeletalMeshLocation(
    MaskSpawnGroup=false
    Mesh Sampling Type=::NewEnumerator5
    Orientation Sampling=::NewEnumerator1
    Sampled Position Offset=<1 0 1.5>
    Skeletal Mesh=DataInterface() {propertyValues {
      DefaultMesh=null
      ExcludeBoneName=root
      OptionalComponentClass=null
      PreviewMesh=&Season_32_Glider
      SkinningMode=SkinOnTheFly
      SoftSourceActor=null
      SourceMode=Default
      UvSetIndex=0
      WholeMeshLOD=-1
      bExcludeBone=false
      bReadDeformedGeometry=true
      bRequireCurrentFrameData=true
      ComponentTags []
      FilteredBones []
      FilteredSockets [FX_Socket_01,FX_Socket_02]
      MeshUserParameter {parameter {name=None, typeDefHandle {registeredTypeIndex=1}, varData []}}
      SamplingRegions []
     }}
    Socket Sampling Mode=::NewEnumerator5
    Use Sampled Mesh Attribute as Ribbon ID=false
   )
   &DynamicMaterialParameters(
    Index 0 Param 1=0.2
    Index 0 Param 2=0.02
    Module.Param0WriteEnabled=true
    Module.Param1WriteEnabled=true
    Module.Param2WriteEnabled=true
    Module.Param3WriteEnabled=true
   )
   &ScaleColor(
    Scale Alpha=&FloatFromCurve(
     FloatCurve=DataInterface() {propertyValues {Curve {keys [{interpMode=RCIM_Cubic, tangentMode=RCTM_User, time=0, value=0}, {interpMode=RCIM_Cubic, tangentMode=RCTM_User, time=0.5, value=1}]}}}
     CurveIndex=$Emitter.Age
    )
    Scale RGB=&FN_Select_Vector_from_Time_Of_Day(Morning=<1 1 1>, Day=<1 1 1>, Evening=<1 1 1>, Night=<0.6 1 1>)
    ScaleA=true
    ScaleRGB=true
   ) {module=ScaleColor001}
   &FN_Light_Attributes(
    Color=&MultiplyLinearColor(Linear Color=$Particles.Color, Scale Factor=<0.25 1 1.5 1>)
    Create Color Attribute=true
    Create Exponent Attribute=true
    Create Radius Attribute=true
    Enable Particle Light Enable=true
    Exponent=1
    Light Enabled=true
    Radius=20
    Scale Color=20
   )
  ]
  rendererClasses [&NiagaraSpriteRendererProperties, &NiagaraLightRendererProperties]
  renderers [
   {rendererClass=&NiagaraSpriteRendererProperties}
   {rendererClass=&NiagaraLightRendererProperties, rendererIndex=1}
  ]
 }
 Ribbon {
  emitterUpdateScript [
   &EmitterState(Life Cycle Mode=::NewEnumerator1, Loop Behavior=::NewEnumerator1, Loop Duration Mode=::NewEnumerator1, UseLoopDelay=false)
   &SpawnBurst_Instantaneous(Spawn Count=50, Use Spawn Probability=false, UseLoopCountLimit=false) {module=SpawnBurst_Instantaneous001}
  ]
  particleSpawnScript [
   &InitializeParticle(
    Color=<rgba 0.6445 0.1981 0.6514 1>
    Lifetime=10
    Ribbon Width=3
    Ribbon Width Mode=::NewEnumerator1
    Sprite Rotation Mode=::NewEnumerator2
    Sprite Size Mode=::NewEnumerator3
    Uniform Sprite Size=5
    UsePositionOffset=false
    Write Material Random=true
   )
   &SkeletalMeshLocation(
    MaskSpawnGroup=false
    Mesh Sampling Type=::NewEnumerator5
    Orientation Sampling=::NewEnumerator1
    Sampled Position Offset=<0 0 -6.5>
    Skeletal Mesh=DataInterface() {propertyValues {
      DefaultMesh=null
      ExcludeBoneName=root
      OptionalComponentClass=null
      PreviewMesh=&Season_32_Glider
      SkinningMode=SkinOnTheFly
      SoftSourceActor=null
      SourceMode=Default
      UvSetIndex=0
      WholeMeshLOD=-1
      bExcludeBone=false
      bReadDeformedGeometry=true
      bRequireCurrentFrameData=true
      ComponentTags []
      FilteredBones []
      FilteredSockets [FX_Socket_02]
      MeshUserParameter {parameter {name=None, typeDefHandle {registeredTypeIndex=1}, varData []}}
      SamplingRegions []
     }}
    Socket Sampling Mode=::NewEnumerator5
    Use Sampled Mesh Attribute as Ribbon ID=false
   )
   &ShapeLocation(
    MaskSpawnGroup=false
    Override Seed=false
    Radius Position=0.6
    Ring / Disc Distribution Mode:ENiagaraSphereDistributionMode=::NewEnumerator1
    Rotation Angle=0.475
    Rotation Axis=<0 0 1>
    Rotation Mode=::NewEnumerator1
    Shape Primitive=::NewEnumerator4
    U Position=&ReturnNormalizedExecIndex(Normalized Index Scale=0.35)
    UseOwnerScale=true
   )
  ]
  particleUpdateScript [
   &ParticleState(Kill Particles When Lifetime Has Elapsed=false, Let Infinitely Lived Particles Die When Emitter Deactivates=true)
   &SkeletalMeshLocation(
    MaskSpawnGroup=false
    Mesh Sampling Type=::NewEnumerator5
    Orientation Sampling=::NewEnumerator1
    Skeletal Mesh=DataInterface() {propertyValues {
      DefaultMesh=null
      ExcludeBoneName=root
      OptionalComponentClass=null
      PreviewMesh=&Season_32_Glider
      SkinningMode=SkinOnTheFly
      SoftSourceActor=null
      SourceMode=Default
      UvSetIndex=0
      WholeMeshLOD=-1
      bExcludeBone=false
      bReadDeformedGeometry=true
      bRequireCurrentFrameData=true
      ComponentTags []
      FilteredBones []
      FilteredSockets [FX_Socket_02]
      MeshUserParameter {parameter {name=None, typeDefHandle {registeredTypeIndex=1}, varData []}}
      SamplingRegions []
     }}
    Socket Sampling Mode=::NewEnumerator5
    Write To Particle Position=false
   ) {module=SkeletalMeshLocation003}
   &ScaleColor(
    Scale Alpha=&Multiply_Float(
     A=&FloatFromCurve(
      FloatCurve=DataInterface() {propertyValues {Curve {keys [
          <at 0 0 RCIM_Cubic>
          {arriveTangentWeight=0.0498, interpMode=RCIM_Cubic, leaveTangentWeight=0.3986, tangentMode=RCTM_User, time=0.05, value=1}
          {arriveTangentWeight=0.2667, interpMode=RCIM_Cubic, leaveTangentWeight=0.0333, tangentMode=RCTM_User, time=0.9, value=1}
          {tangentMode=RCTM_User, time=1, value=0}
         ]}}}
      CurveIndex=&ReturnNormalizedExecIndex()
     )
     B=&FloatFromCurve(
      FloatCurve=DataInterface() {propertyValues {Curve {keys [{interpMode=RCIM_Cubic, tangentMode=RCTM_User, time=1.8, value=0}, {interpMode=RCIM_Cubic, tangentMode=RCTM_User, time=2.8, value=1}]}}}
      CurveIndex=&Subtract_Float(A=$Emitter.Age, B=&ReturnNormalizedExecIndex())
     )
    )
    Scale RGB=&Multiply_VectorByFloat(
     Vector=&Multiply_Vector(
      A=&MakeVectorFromLinearColorRGB(Color=&ColorFromCurve(
        ColorCurve=DataInterface() {propertyValues {
          AlphaCurve {keys [<at 0 1 RCIM_Cubic>]}
          BlueCurve {keys [
            <at 0 1>
            {arriveTangent=-1.293, interpMode=RCIM_Cubic, leaveTangent=-1.293, time=0.0764, value=0.6514}
            {arriveTangent=-0.57, interpMode=RCIM_Cubic, leaveTangent=-0.57, time=0.6767, value=0.125}
            <at 1 0.125>
           ]}
          GreenCurve {keys [
            <at 0 0.4375>
            {arriveTangent=-0.6331, interpMode=RCIM_Cubic, leaveTangent=-0.6331, time=0.0764, value=0.1981}
            {arriveTangent=-0.2102, interpMode=RCIM_Cubic, leaveTangent=-0.2102, time=0.6767, value=0.009}
            <at 1 0.0039>
           ]}
          RedCurve {keys [
            <at 0 0.9914>
            {arriveTangent=-1.3604, interpMode=RCIM_Cubic, leaveTangent=-1.3604, time=0.0764, value=0.6445}
            {arriveTangent=-0.6649, interpMode=RCIM_Cubic, leaveTangent=-0.6649, time=0.6767, value=0.0708}
            <at 1 0.0304>
           ]}
         }}
        CurveIndex=&ReturnNormalizedExecIndex()
       ))
      B=&Multiply_Vector(
       A=<0.4 1 1.25>
       B=&FN_Select_Vector_from_Time_Of_Day(Morning=<1 1 1>, Day=<1 1 1>, Evening=<1 1 1>, Night=<0.7 1 1>)
      )
     )
     Float=10
    )
    ScaleA=true
    ScaleRGB=true
   )
   /SetVariables(Particles.PositionFinal:NiagaraPosition=&AddVectorToPosition(
     Position=$Particles.Position
     Vector=&AddVectorToPosition(
      Position=&Add_Vector(
       A=&Multiply_VectorByFloat(
        Vector=&Subtract_Position(A=$Particles.Position, B=$Particles.SkeletalMeshLocation003.SampledPosition)
        Float=&FloatFromCurve(
         FloatCurve=DataInterface() {propertyValues {
           bOptimizeLUT=false
           bUseLUT=false
           Curve {keys [
             {interpMode=RCIM_Cubic, leaveTangentWeight=0.15, tangentMode=RCTM_User, time=0.05, value=-0.05}
             {arriveTangentWeight=0.2905, interpMode=RCIM_Cubic, tangentMode=RCTM_User, time=0.2, value=0}
             {arriveTangentWeight=0.2685, interpMode=RCIM_Cubic, tangentMode=RCTM_User, time=1, value=0}
            ]}
          }}
         CurveIndex=&ReturnNormalizedExecIndex()
        )
       )
       B=<0 0 5>
      )
      Vector=&Multiply_VectorByFloat(
       Vector=&Subtract_Position(A=$Particles.Position, B=$Particles.SkeletalMeshLocation003.SampledPosition)
       Float=&Multiply_Float(
        A=&FloatFromCurve(
         FloatCurve=DataInterface() {propertyValues {
           bOptimizeLUT=false
           bUseLUT=false
           Curve {
            postInfinityExtrap=RCCE_Cycle
            preInfinityExtrap=RCCE_Cycle
            keys [
             {arriveTangent=48.477, interpMode=RCIM_Cubic, leaveTangent=48.477, leaveTangentWeight=0.0167, tangentMode=RCTM_User, time=0.05, value=0}
             {arriveTangentWeight=0.2506, interpMode=RCIM_Cubic, leaveTangentWeight=0.2506, tangentMode=RCTM_User, time=0.1, value=1}
             {arriveTangentWeight=0.0236, interpMode=RCIM_Cubic, leaveTangentWeight=0.0236, tangentMode=RCTM_User, time=0.18, value=-3}
             {arriveTangentWeight=0.0964, interpMode=RCIM_Cubic, leaveTangentWeight=0.0964, tangentMode=RCTM_User, time=0.25, value=0.5}
             {arriveTangentWeight=0.0639, interpMode=RCIM_Cubic, leaveTangentWeight=0.0639, tangentMode=RCTM_User, time=0.3, value=0}
             {arriveTangentWeight=0.1364, interpMode=RCIM_Cubic, leaveTangentWeight=0.1364, tangentMode=RCTM_User, time=0.35, value=1}
             {arriveTangentWeight=0.03, interpMode=RCIM_Cubic, leaveTangentWeight=0.03, tangentMode=RCTM_User, time=0.4, value=-2}
             {arriveTangentWeight=0.0498, interpMode=RCIM_Cubic, leaveTangentWeight=0.0498, tangentMode=RCTM_User, time=0.45, value=0.1131}
             {
              arriveTangent=23.0731
              arriveTangentWeight=0.0431
              interpMode=RCIM_Cubic
              leaveTangent=23.0731
              leaveTangentWeight=0.0431
              tangentMode=RCTM_User
              time=0.5
              value=-1
             }
             {arriveTangent=14.6983, interpMode=RCIM_Cubic, leaveTangent=14.6983, time=0.5587, value=-1.4642}
             {arriveTangent=-11.1972, interpMode=RCIM_Cubic, leaveTangent=-11.1972, time=0.5941, value=0.3832}
             {arriveTangent=-0.2609, interpMode=RCIM_Cubic, leaveTangent=-0.2609, time=0.6542, value=-2.5326}
             {arriveTangent=15.9984, interpMode=RCIM_Cubic, leaveTangent=15.9984, time=0.7173, value=0.3511}
             {arriveTangent=6.8144, interpMode=RCIM_Cubic, leaveTangent=6.8144, time=0.7611, value=-0.8219}
             {arriveTangent=-4.195, interpMode=RCIM_Cubic, leaveTangent=-4.195, time=0.8022, value=0.9299}
             {arriveTangent=-4.7018, interpMode=RCIM_Cubic, leaveTangent=-4.7018, time=0.9059, value=-1.4292}
             {arriveTangent=10.1164, interpMode=RCIM_Cubic, leaveTangent=10.1164, tangentMode=RCTM_User, time=1, value=0}
            ]
           }
          }}
         CurveIndex=&Add_Float(
          A=&Multiply_Float(A=&ReturnNormalizedExecIndex(), B=1.25)
          B=&Multiply_Float(A=$Engine.Time, B=-0.68)
         )
         Scale Curve=&Multiply_Float(A=0.05, B=1)
        )
        B=&FloatFromCurve(
         FloatCurve=DataInterface() {propertyValues {Curve {keys [
             {interpMode=RCIM_Cubic, tangentMode=RCTM_User, time=0, value=0}
             <at 0.05 1 RCIM_Cubic>
             {arriveTangent=-2.3266, interpMode=RCIM_Cubic, leaveTangent=-2.3266, tangentMode=RCTM_User, time=0.6, value=0}
            ]}}}
         CurveIndex=&ReturnNormalizedExecIndex()
        )
       )
      )
     )
    )) {isSetParameters=true}
   &ScaleRibbonWidth(Ribbon Width Scale=&FloatFromCurve(
     FloatCurve=DataInterface() {propertyValues {Curve {keys [
         {interpMode=RCIM_Cubic, tangentMode=RCTM_User, time=0, value=0.2}
         {arriveTangent=-0.2, interpMode=RCIM_Cubic, leaveTangent=-0.2, time=0.1, value=1}
         {interpMode=RCIM_Cubic, tangentMode=RCTM_User, time=1, value=0}
        ]}}}
     CurveIndex=&ReturnNormalizedExecIndex()
     Scale Curve=0.5
    ))
   /SetVariables(
    Particles.RibbonColor_02:LinearColor=&MultiplyLinearColor(Linear Color=$Particles.Color, Scale Factor=<1 1 1 0.35>)
    Particles.RibbonWidth_02:NiagaraFloat=&Add_Float(
     A=&Multiply_Float(A=$Particles.RibbonWidth, B=8)
     B=&FloatFromCurve(
      FloatCurve=DataInterface() {propertyValues {Curve {keys [{tangentMode=RCTM_User, time=0.1, value=0}, {tangentMode=RCTM_User, time=0.3, value=1}]}}}
      CurveIndex=&ReturnNormalizedExecIndex()
      Scale Curve=5
     )
    )
   ) {isSetParameters=true, module=SetVariables_#1}
   /SetVariables() {isSetParameters=true, module=SetVariables_#2}
  ]
  rendererClasses [&NiagaraRibbonRendererProperties]
  renderers [
   {rendererClass=&NiagaraRibbonRendererProperties}
   {rendererClass=&NiagaraRibbonRendererProperties, rendererIndex=1}
  ]
 }
 Ribbon_lowmed {
  emitterUpdateScript [
   &EmitterState(Life Cycle Mode=::NewEnumerator1, Loop Behavior=::NewEnumerator1, Loop Duration Mode=::NewEnumerator1, UseLoopDelay=false)
   &SpawnBurst_Instantaneous(Spawn Count=20, Use Spawn Probability=false, UseLoopCountLimit=false) {module=SpawnBurst_Instantaneous001}
  ]
  particleSpawnScript [
   &InitializeParticle(
    Color=<rgba 0.6445 0.1981 0.6514 1>
    Lifetime=10
    Ribbon Width=3
    Ribbon Width Mode=::NewEnumerator1
    Sprite Rotation Mode=::NewEnumerator2
    Sprite Size Mode=::NewEnumerator3
    Uniform Sprite Size=5
    UsePositionOffset=false
    Write Material Random=true
   )
   &SkeletalMeshLocation(
    MaskSpawnGroup=false
    Mesh Sampling Type=::NewEnumerator5
    Orientation Sampling=::NewEnumerator1
    Sampled Position Offset=<0 0 -6.5>
    Skeletal Mesh=DataInterface() {propertyValues {
      DefaultMesh=null
      ExcludeBoneName=root
      OptionalComponentClass=null
      PreviewMesh=&Season_32_Glider
      SkinningMode=SkinOnTheFly
      SoftSourceActor=null
      SourceMode=Default
      UvSetIndex=0
      WholeMeshLOD=-1
      bExcludeBone=false
      bReadDeformedGeometry=true
      bRequireCurrentFrameData=true
      ComponentTags []
      FilteredBones []
      FilteredSockets [FX_Socket_02]
      MeshUserParameter {parameter {name=None, typeDefHandle {registeredTypeIndex=1}, varData []}}
      SamplingRegions []
     }}
    Socket Sampling Mode=::NewEnumerator5
    Use Sampled Mesh Attribute as Ribbon ID=false
   )
   &ShapeLocation(
    MaskSpawnGroup=false
    Override Seed=false
    Radius Position=0.6
    Ring / Disc Distribution Mode:ENiagaraSphereDistributionMode=::NewEnumerator1
    Rotation Angle=0.475
    Rotation Axis=<0 0 1>
    Rotation Mode=::NewEnumerator1
    Shape Primitive=::NewEnumerator4
    U Position=&ReturnNormalizedExecIndex(Normalized Index Scale=0.35)
    UseOwnerScale=true
   )
  ]
  particleUpdateScript [
   &ParticleState(Kill Particles When Lifetime Has Elapsed=false, Let Infinitely Lived Particles Die When Emitter Deactivates=true)
   &SkeletalMeshLocation(
    MaskSpawnGroup=false
    Mesh Sampling Type=::NewEnumerator5
    Orientation Sampling=::NewEnumerator1
    Skeletal Mesh=DataInterface() {propertyValues {
      DefaultMesh=null
      ExcludeBoneName=root
      OptionalComponentClass=null
      PreviewMesh=&Season_32_Glider
      SkinningMode=SkinOnTheFly
      SoftSourceActor=null
      SourceMode=Default
      UvSetIndex=0
      WholeMeshLOD=-1
      bExcludeBone=false
      bReadDeformedGeometry=true
      bRequireCurrentFrameData=true
      ComponentTags []
      FilteredBones []
      FilteredSockets [FX_Socket_02]
      MeshUserParameter {parameter {name=None, typeDefHandle {registeredTypeIndex=1}, varData []}}
      SamplingRegions []
     }}
    Socket Sampling Mode=::NewEnumerator5
    Write To Particle Position=false
   ) {module=SkeletalMeshLocation003}
   &ScaleColor(
    Scale Alpha=&Multiply_Float(
     A=&FloatFromCurve(
      FloatCurve=DataInterface() {propertyValues {Curve {keys [
          <at 0 0 RCIM_Cubic>
          {arriveTangentWeight=0.0498, interpMode=RCIM_Cubic, leaveTangentWeight=0.3986, tangentMode=RCTM_User, time=0.05, value=1}
          {arriveTangentWeight=0.2667, interpMode=RCIM_Cubic, leaveTangentWeight=0.0333, tangentMode=RCTM_User, time=0.9, value=1}
          {tangentMode=RCTM_User, time=1, value=0}
         ]}}}
      CurveIndex=&ReturnNormalizedExecIndex()
     )
     B=&FloatFromCurve(
      FloatCurve=DataInterface() {propertyValues {Curve {keys [{interpMode=RCIM_Cubic, tangentMode=RCTM_User, time=1.8, value=0}, {interpMode=RCIM_Cubic, tangentMode=RCTM_User, time=2.8, value=1}]}}}
      CurveIndex=&Subtract_Float(A=$Emitter.Age, B=&ReturnNormalizedExecIndex())
     )
    )
    Scale RGB=&Multiply_VectorByFloat(
     Vector=&Multiply_Vector(
      A=&MakeVectorFromLinearColorRGB(Color=&ColorFromCurve(
        ColorCurve=DataInterface() {propertyValues {
          AlphaCurve {keys [<at 0 1 RCIM_Cubic>]}
          BlueCurve {keys [
            <at 0 1>
            {arriveTangent=-1.293, interpMode=RCIM_Cubic, leaveTangent=-1.293, time=0.0764, value=0.6514}
            {arriveTangent=-0.57, interpMode=RCIM_Cubic, leaveTangent=-0.57, time=0.6767, value=0.125}
            <at 1 0.125>
           ]}
          GreenCurve {keys [
            <at 0 0.4375>
            {arriveTangent=-0.6331, interpMode=RCIM_Cubic, leaveTangent=-0.6331, time=0.0764, value=0.1981}
            {arriveTangent=-0.2102, interpMode=RCIM_Cubic, leaveTangent=-0.2102, time=0.6767, value=0.009}
            <at 1 0.0039>
           ]}
          RedCurve {keys [
            <at 0 0.9914>
            {arriveTangent=-1.3604, interpMode=RCIM_Cubic, leaveTangent=-1.3604, time=0.0764, value=0.6445}
            {arriveTangent=-0.6649, interpMode=RCIM_Cubic, leaveTangent=-0.6649, time=0.6767, value=0.0708}
            <at 1 0.0304>
           ]}
         }}
        CurveIndex=&ReturnNormalizedExecIndex()
       ))
      B=&Multiply_Vector(
       A=<0.4 1 1.25>
       B=&FN_Select_Vector_from_Time_Of_Day(Morning=<1 1 1>, Day=<1 1 1>, Evening=<1 1 1>, Night=<0.7 1 1>)
      )
     )
     Float=10
    )
    ScaleA=true
    ScaleRGB=true
   )
   /SetVariables(Particles.PositionFinal:NiagaraPosition=&AddVectorToPosition(
     Position=$Particles.Position
     Vector=&AddVectorToPosition(
      Position=&Add_Vector(
       A=&Multiply_VectorByFloat(
        Vector=&Subtract_Position(A=$Particles.Position, B=$Particles.SkeletalMeshLocation003.SampledPosition)
        Float=&FloatFromCurve(
         FloatCurve=DataInterface() {propertyValues {
           bOptimizeLUT=false
           bUseLUT=false
           Curve {keys [
             {interpMode=RCIM_Cubic, leaveTangentWeight=0.15, tangentMode=RCTM_User, time=0.05, value=-0.05}
             {arriveTangentWeight=0.2905, interpMode=RCIM_Cubic, tangentMode=RCTM_User, time=0.2, value=0}
             {arriveTangentWeight=0.2685, interpMode=RCIM_Cubic, tangentMode=RCTM_User, time=1, value=0}
            ]}
          }}
         CurveIndex=&ReturnNormalizedExecIndex()
        )
       )
       B=<0 0 5>
      )
      Vector=&Multiply_VectorByFloat(
       Vector=&Subtract_Position(A=$Particles.Position, B=$Particles.SkeletalMeshLocation003.SampledPosition)
       Float=&Multiply_Float(
        A=&FloatFromCurve(
         FloatCurve=DataInterface() {propertyValues {
           bOptimizeLUT=false
           bUseLUT=false
           Curve {
            postInfinityExtrap=RCCE_Cycle
            preInfinityExtrap=RCCE_Cycle
            keys [
             {arriveTangent=48.477, interpMode=RCIM_Cubic, leaveTangent=48.477, leaveTangentWeight=0.0167, tangentMode=RCTM_User, time=0.05, value=0}
             {arriveTangentWeight=0.2506, interpMode=RCIM_Cubic, leaveTangentWeight=0.2506, tangentMode=RCTM_User, time=0.1, value=1}
             {arriveTangentWeight=0.0236, interpMode=RCIM_Cubic, leaveTangentWeight=0.0236, tangentMode=RCTM_User, time=0.18, value=-3}
             {arriveTangentWeight=0.1364, interpMode=RCIM_Cubic, leaveTangentWeight=0.1364, tangentMode=RCTM_User, time=0.3, value=1}
             {arriveTangentWeight=0.03, interpMode=RCIM_Cubic, leaveTangentWeight=0.03, tangentMode=RCTM_User, time=0.4, value=-2}
             {arriveTangentWeight=0.0498, interpMode=RCIM_Cubic, leaveTangentWeight=0.0498, tangentMode=RCTM_User, time=0.5, value=0.1131}
             {arriveTangent=2.7029, interpMode=RCIM_Cubic, leaveTangent=2.7029, time=0.6542, value=-2.5326}
             {arriveTangent=4.3839, interpMode=RCIM_Cubic, leaveTangent=4.3839, time=0.8022, value=0.9299}
             {arriveTangent=-4.7018, interpMode=RCIM_Cubic, leaveTangent=-4.7018, time=0.9059, value=-1.4292}
             {arriveTangent=10.1164, interpMode=RCIM_Cubic, leaveTangent=10.1164, tangentMode=RCTM_User, time=1, value=0}
            ]
           }
          }}
         CurveIndex=&Add_Float(
          A=&Multiply_Float(A=&ReturnNormalizedExecIndex(), B=1)
          B=&Multiply_Float(A=$Engine.Time, B=-0.5)
         )
         Scale Curve=&Multiply_Float(A=0.05, B=1)
        )
        B=&FloatFromCurve(
         FloatCurve=DataInterface() {propertyValues {Curve {keys [
             {interpMode=RCIM_Cubic, tangentMode=RCTM_User, time=0, value=0}
             <at 0.05 1 RCIM_Cubic>
             {arriveTangent=-2.3266, interpMode=RCIM_Cubic, leaveTangent=-2.3266, tangentMode=RCTM_User, time=0.6, value=0}
            ]}}}
         CurveIndex=&ReturnNormalizedExecIndex()
        )
       )
      )
     )
    )) {isSetParameters=true}
   &ScaleRibbonWidth(Ribbon Width Scale=&FloatFromCurve(
     FloatCurve=DataInterface() {propertyValues {Curve {keys [
         {interpMode=RCIM_Cubic, tangentMode=RCTM_User, time=0, value=0.2}
         {arriveTangent=-0.2, interpMode=RCIM_Cubic, leaveTangent=-0.2, time=0.1, value=1}
         {interpMode=RCIM_Cubic, tangentMode=RCTM_User, time=1, value=0}
        ]}}}
     CurveIndex=&ReturnNormalizedExecIndex()
     Scale Curve=0.5
    ))
   /SetVariables(
    Particles.RibbonColor_02:LinearColor=&MultiplyLinearColor(Linear Color=$Particles.Color, Scale Factor=<1 1 1 0.35>)
    Particles.RibbonWidth_02:NiagaraFloat=&Add_Float(
     A=&Multiply_Float(A=$Particles.RibbonWidth, B=8)
     B=&FloatFromCurve(
      FloatCurve=DataInterface() {propertyValues {Curve {keys [{tangentMode=RCTM_User, time=0.1, value=0}, {tangentMode=RCTM_User, time=0.3, value=1}]}}}
      CurveIndex=&ReturnNormalizedExecIndex()
      Scale Curve=5
     )
    )
   ) {isSetParameters=true, module=SetVariables_#1}
   /SetVariables() {isSetParameters=true, module=SetVariables_#2}
  ]
  rendererClasses [&NiagaraRibbonRendererProperties]
  renderers [
   {rendererClass=&NiagaraRibbonRendererProperties}
   {rendererClass=&NiagaraRibbonRendererProperties, rendererIndex=1}
  ]
 }
 SpeedLine {
  emitterUpdateScript [
   &EmitterState()
   &SpawnRate(
    SpawnRate=&MakeCustomFloatFromBool(Bool=$User.bIsFrontEndPreview, True Float=0, False Float=90)
    Use Spawn Probability=false
   ) {module=SpawnRate001}
  ]
  particleSpawnScript [
   &InitializeParticle(
    Color=&MakeLinearColorFromVectorAndFloat(
     "Vector (RGB)"=<1 1 1>
     "Float (Alpha)"=&Multiply_Float(
      A=&UniformRangedFloat(Minimum=0.1, Maximum=0.18)
      B=&MakeCustomFloatFromBool(Bool=$User.bIsFrontEndPreview, True Float=0.5, False Float=1)
     )
    )
    Lifetime Max=0.16
    Lifetime Min=0.1
    Lifetime Mode=::NewEnumerator1
    Sprite Size Max=<3 10>
    Sprite Size Min=<2 2>
    Sprite Size Mode=::NewEnumerator2
    UsePositionOffset=false
    Write Material Random=true
   )
   &TorusLocation(
    Align Torus With Axis=true
    Handle Radius=0
    Large Radius:NiagaraFloat=&RandomRangeFloat(Minimum=60, Maximum=160)
    MaskSpawnGroup=false
    Module.Override Randomness=false
    Offset=<0 -180 0>
    Torus Axis=<0 1 0>
    Torus Origin:NiagaraPosition=&Add_Vector(
     A=Linked() {linkedVariable {name=Particles.Position, type=NiagaraPosition}}
     B=&MakeVector(X=0, Y=&MakeCustomFloatFromBool(Bool=$User.bIsFrontEndPreview, True Float=200, False Float=400), Z=0)
    )
   )
   &AddVelocityFromPoint(
    Module.Use Falloff Distance=false
    Origin Offset=<0 500 0>
    Velocity Strength=&Multiply_Float(A=600, B=&MakeCustomFloatFromBool(Bool=$User.bIsFrontEndPreview, True Float=-1, False Float=1))
   )
   &AddVelocity(Velocity=&Multiply_Vector(
     A=&RandomRangeVector(Minimum=<0 300 0>, Maximum=<0 700 0>)
     B=&MakeVector(X=1, Y=&MakeCustomFloatFromBool(Bool=$User.bIsFrontEndPreview, True Float=-1, False Float=1), Z=1)
    )) {enabled=false}
  ]
  particleUpdateScript [
   &SolveForcesAndVelocity(Clamp Velocity=false, Limit Acceleration=false)
   &ParticleState()
   &ScaleSpriteSize(Scale Factor=&Vector2DFromCurve(
     Vector2Curve=DataInterface() {propertyValues {XCurve {keys [<at 0 0.5 RCIM_Cubic>,<at 0.5 1 RCIM_Cubic>]}, YCurve {keys [<at 0 0 RCIM_Cubic>,<at 0.5 1 RCIM_Cubic>]}}}
     ScaleCurve=&Vector2DFromFloat(Value=1)
    ))
   &ScaleSpriteSizeBySpeed(Max Scale Factor=<1 50>, Min Scale Factor=<0.01 0.05>, Velocity Threshold=1)
   &ScaleColor(
    Scale Alpha=&FloatFromCurve(
     FloatCurve=DataInterface() {propertyValues {Curve {keys [
         <at 0 0 RCIM_Cubic>
         <at 0.3759 1.0207>
         {arriveTangent=-2.221, interpMode=RCIM_Cubic, leaveTangent=-2.221, tangentMode=RCTM_User, time=1, value=0}
        ]}}}
     Scale Curve=&MakeCustomFloatFromBool(
      Bool=$User.bIsFrontEndPreview
      True Float=&MakeCustomFloatFromBool(Bool=$User.bIsFullyDeployed, True Float=1, False Float=0)
      False Float=&Lerp_Float(A=0, B=1, Alpha=$System.Timeline001.Time)
     )
    )
    Scale RGB=&VectorFromFloat(Value=1)
    ScaleA=true
    ScaleRGB=true
   )
  ]
  rendererClasses [&NiagaraSpriteRendererProperties]
  renderers [{rendererClass=&NiagaraSpriteRendererProperties}]
 }
}
summary {
 systemName=NS_Glider_RankedFirth_Trail_02
 emitters [
  {emitterName=Ribbon, rendererClasses [&NiagaraRibbonRendererProperties]}
  {
   emitterName=BlowingParticles001
   rendererClasses [&NiagaraSpriteRendererProperties, &NiagaraLightRendererProperties]
  }
  {emitterName=SpeedLine, rendererClasses [&NiagaraSpriteRendererProperties]}
  {emitterName=Ribbon_lowmed, rendererClasses [&NiagaraRibbonRendererProperties]}
 ]
 userVariables [
  User.bIsFrontEnd {type=NiagaraBool}
  User.bIsFrontEndPreview {type=NiagaraBool}
  User.bIsFullyDeployed {type=NiagaraBool}
  User.ForwardDot {type=NiagaraFloat, defaultValue=0, description=""}
 ]
}
